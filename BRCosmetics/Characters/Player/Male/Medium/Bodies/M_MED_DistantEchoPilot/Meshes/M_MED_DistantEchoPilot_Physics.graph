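# Physics asset for the DistantEchoPilot medium male player body, built around the dynamic bones of its jacket.
# Six jacket chains (a–f) each run from a tiny master capsule through three capsules joined by swing and narrow twist limits.
# The masters attach to spine_03 with free angular constraints, as does a left-only upper arm capsule.
# Pelvis, spine and thigh capsules cover the body, and a free constraint cross-links the c and d chain tips.
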PhysicsAsset /BRCosmetics/Characters/Player/Male/Medium/Bodies/M_MED_DistantEchoPilot/Meshes/M_MED_DistantEchoPilot_Physics
bodies {
 dyn_jacket_a_01 [{shapeType=Capsule, length=4.25, radius=3.75, shapeName=dyn_jacket_a_01_capsule, center=<3 0 2>, rotation=<rot 90 90 0>}]
 dyn_jacket_a_02 [{shapeType=Capsule, length=7, radius=4, shapeName=dyn_jacket_a_02_capsule, center=<3 0 4>, rotation=<rot 90 -165.9638 104.0362>}]
 dyn_jacket_a_03 [{shapeType=Capsule, length=13.5, radius=5, shapeName=dyn_jacket_f_03_capsule, center=<3 0 4>, rotation=<rot 90 90 0>}]
 dyn_jacket_a_mstr [{shapeType=Capsule, length=0.505, radius=0.505, shapeName=dyn_jacket_a_mstr_capsule, rotation=<rot 0 -90 0>}]
 dyn_jacket_b_01 [{shapeType=Capsule, length=7.75, radius=4.5, shapeName=dyn_jacket_b_01_capsule, center=<3 0 1.5>, rotation=<rot 90 -165.9638 104.0362>}]
 dyn_jacket_b_02 [{shapeType=Capsule, length=7.75, radius=5, shapeName=dyn_jacket_b_02_capsule, center=<3 0 2>, rotation=<rot 90 180 90>}]
 dyn_jacket_b_03 [{shapeType=Capsule, length=10.25, radius=5, shapeName=dyn_jacket_b_03_capsule, center=<3 0 2>, rotation=<rot 90 90 0>}]
 dyn_jacket_b_mstr [{shapeType=Capsule, length=0.505, radius=0.505, shapeName=dyn_jacket_b_mstr_capsule, rotation=<rot 0 -90 0>}]
 dyn_jacket_c_01 [{shapeType=Capsule, length=4, radius=4, shapeName=dyn_jacket_c_01_capsule, center=<3 0 2.5>, rotation=<rot 90 90 0>}]
 dyn_jacket_c_02 [{shapeType=Capsule, length=4, radius=4, shapeName=dyn_jacket_c_02_capsule, center=<3 0 2.5>, rotation=<rot 90 90 0>}]
 dyn_jacket_c_03 [{shapeType=Capsule, length=6.25, radius=3.75, shapeName=dyn_jacket_c_03_capsule, center=<3 0 3.5>, rotation=<rot 90 -165.9638 104.0362>}]
 dyn_jacket_c_mstr [{shapeType=Capsule, length=0.505, radius=0.505, shapeName=dyn_jacket_c_mstr_capsule, rotation=<rot 0 -90 0>}]
 dyn_jacket_d_01 [{shapeType=Capsule, length=4, radius=4, shapeName=dyn_jacket_d_01_capsule, center=<3 0 2.5>, rotation=<rot 90 90 0>}]
 dyn_jacket_d_02 [{shapeType=Capsule, length=4, radius=4, shapeName=dyn_jacket_d_02_capsule, center=<3 0 2.5>, rotation=<rot 90 90 0>}]
 dyn_jacket_d_03 [{shapeType=Capsule, length=6.25, radius=3.75, shapeName=dyn_jacket_d_03_capsule, center=<3 0 3.5>, rotation=<rot 90 -165.9638 104.0362>}]
 dyn_jacket_d_mstr [{shapeType=Capsule, length=0.505, radius=0.505, shapeName=dyn_jacket_d_mstr_capsule, rotation=<rot 0 -90 0>}]
 dyn_jacket_e_01 [{shapeType=Capsule, length=4, radius=5, shapeName=dyn_jacket_e_01_capsule, center=<3 0 2>, rotation=<rot 90 90 0>}]
 dyn_jacket_e_02 [{shapeType=Capsule, length=4, radius=5, shapeName=dyn_jacket_e_02_capsule, center=<3 0 2>, rotation=<rot 90 90 0>}]
 dyn_jacket_e_03 [{shapeType=Capsule, length=4, radius=5.5, shapeName=dyn_jacket_e_03_capsule, center=<3 0 2>, rotation=<rot 90 90 0>}]
 dyn_jacket_e_mstr [{shapeType=Capsule, length=0.505, radius=0.505, shapeName=dyn_jacket_e_mstr_capsule, rotation=<rot 0 -90 0>}]
 dyn_jacket_f_01 [{shapeType=Capsule, length=4, radius=4, shapeName=dyn_jacket_f_01_capsule, center=<3 0 2>, rotation=<rot 90 90 0>}]
 dyn_jacket_f_02 [{shapeType=Capsule, length=4, radius=4, shapeName=dyn_jacket_f_02_capsule, center=<3 0 2>, rotation=<rot 90 90 0>}]
 dyn_jacket_f_03 [{shapeType=Capsule, length=13.5, radius=5, shapeName=dyn_jacket_f_03_capsule, center=<3 0 2>, rotation=<rot 90 90 0>}]
 dyn_jacket_f_mstr [{shapeType=Capsule, length=0.505, radius=0.505, shapeName=dyn_jacket_f_mstr_capsule, rotation=<rot 0 -90 0>}]
 pelvis [{shapeType=Capsule, length=9.6871, radius=10.6798, shapeName=pelvis_capsule}]
 spine_01 [{shapeType=Capsule, length=5, radius=10.5, shapeName=spine_01_capsule, center=<0 1.1149 0>}]
 spine_02 [{shapeType=Capsule, length=4.5, radius=10.5, shapeName=spine_02_capsule, center=<0 0.5719 0>}]
 spine_03 [{shapeType=Capsule, length=7.25, radius=9.75, shapeName=spine_03_capsule, center=<0 -1.0192 0>}]
 thigh_l [{shapeType=Capsule, length=31.3476, radius=9.7327, shapeName=thigh_l_capsule, center=<-18.2547 1.3748 0>, rotation=<rot 90 172.875 -11.2768>}]
 thigh_r [{shapeType=Capsule, length=32.5976, radius=8.4827, shapeName=thigh_l_capsule, center=<18.1043 -0.7912 0>, rotation=<rot 90 89.6476 -89.8727>}]
 upperarm_l [{
   shapeType=Capsule
   length=6.4082
   radius=4.5499
   shapeName=upperarm_l_capsule
   center=<19.9962 -0.1695 0.3977>
   rotation=<rot -23.0864 -89.9999 85.3852>
  }]
}
constraints [
 {bone1Name=dyn_jacket_a_02, bone2Name=dyn_jacket_a_01, swing1LimitDegrees=15, swing2LimitDegrees=55, twistLimitDegrees=5}
 {bone1Name=dyn_jacket_a_03, bone2Name=dyn_jacket_a_02, swing1LimitDegrees=20, swing2LimitDegrees=50, twistLimitDegrees=2}
 {bone1Name=dyn_jacket_b_02, bone2Name=dyn_jacket_b_01, swing1LimitDegrees=30, swing2LimitDegrees=30, twistLimitDegrees=2}
 {bone1Name=dyn_jacket_b_03, bone2Name=dyn_jacket_b_02, swing1LimitDegrees=30, swing2LimitDegrees=30, twistLimitDegrees=2}
 {bone1Name=dyn_jacket_c_02, bone2Name=dyn_jacket_c_01, swing1LimitDegrees=10, swing2LimitDegrees=25, twistLimitDegrees=2}
 {bone1Name=dyn_jacket_c_03, bone2Name=dyn_jacket_c_02, swing1LimitDegrees=20, swing2LimitDegrees=45, twistLimitDegrees=2}
 {bone1Name=dyn_jacket_d_02, bone2Name=dyn_jacket_d_01, swing1LimitDegrees=10, swing2LimitDegrees=25, twistLimitDegrees=2}
 {bone1Name=dyn_jacket_d_03, bone2Name=dyn_jacket_d_02, swing1LimitDegrees=20, swing2LimitDegrees=45, twistLimitDegrees=2}
 {bone1Name=dyn_jacket_e_03, bone2Name=dyn_jacket_e_02, swing1LimitDegrees=30, swing2LimitDegrees=30, twistLimitDegrees=2}
 {bone1Name=dyn_jacket_e_02, bone2Name=dyn_jacket_e_01, swing1LimitDegrees=30, swing2LimitDegrees=30, twistLimitDegrees=2}
 {bone1Name=dyn_jacket_f_03, bone2Name=dyn_jacket_f_02, swing1LimitDegrees=30, swing2LimitDegrees=30, twistLimitDegrees=5}
 {bone1Name=dyn_jacket_f_02, bone2Name=dyn_jacket_f_01, swing1LimitDegrees=30, swing2LimitDegrees=30, twistLimitDegrees=5}
 {bone1Name=dyn_jacket_a_01, bone2Name=dyn_jacket_a_mstr, swing1LimitDegrees=10, swing2LimitDegrees=50, twistLimitDegrees=5}
 {bone1Name=dyn_jacket_b_01, bone2Name=dyn_jacket_b_mstr, swing1LimitDegrees=5, swing2LimitDegrees=35, twistLimitDegrees=2}
 {bone1Name=dyn_jacket_c_01, bone2Name=dyn_jacket_c_mstr, swing1LimitDegrees=5, swing2LimitDegrees=20, twistLimitDegrees=2}
 {bone1Name=dyn_jacket_d_01, bone2Name=dyn_jacket_d_mstr, swing1LimitDegrees=5, swing2LimitDegrees=20, twistLimitDegrees=2}
 {bone1Name=dyn_jacket_e_01, bone2Name=dyn_jacket_e_mstr, swing1LimitDegrees=5, swing2LimitDegrees=20, twistLimitDegrees=2}
 {bone1Name=dyn_jacket_f_01, bone2Name=dyn_jacket_f_mstr, swing1LimitDegrees=15, swing2LimitDegrees=20, twistLimitDegrees=5}
 {
  bone1Name=spine_01
  bone2Name=pelvis
  swing1LimitDegrees=45
  swing1Motion=Free
  swing2LimitDegrees=45
  swing2Motion=Free
  twistLimitDegrees=45
  twistMotion=Free
 }
 {
  bone1Name=spine_02
  bone2Name=spine_01
  swing1LimitDegrees=45
  swing1Motion=Free
  swing2LimitDegrees=45
  swing2Motion=Free
  twistLimitDegrees=45
  twistMotion=Free
 }
 {
  bone1Name=spine_03
  bone2Name=spine_02
  swing1LimitDegrees=45
  swing1Motion=Free
  swing2LimitDegrees=45
  swing2Motion=Free
  twistLimitDegrees=45
  twistMotion=Free
 }
 {
  bone1Name=thigh_l
  bone2Name=pelvis
  swing1LimitDegrees=45
  swing1Motion=Free
  swing2LimitDegrees=45
  swing2Motion=Free
  twistLimitDegrees=45
  twistMotion=Free
 }
 {
  bone1Name=dyn_jacket_a_mstr
  bone2Name=spine_03
  swing1LimitDegrees=45
  swing1Motion=Free
  swing2LimitDegrees=45
  swing2Motion=Free
  twistLimitDegrees=45
  twistMotion=Free
 }
 {
  bone1Name=dyn_jacket_b_mstr
  bone2Name=spine_03
  swing1LimitDegrees=45
  swing1Motion=Free
  swing2LimitDegrees=45
  swing2Motion=Free
  twistLimitDegrees=45
  twistMotion=Free
 }
 {
  bone1Name=dyn_jacket_c_mstr
  bone2Name=spine_03
  swing1LimitDegrees=45
  swing1Motion=Free
  swing2LimitDegrees=45
  swing2Motion=Free
  twistLimitDegrees=45
  twistMotion=Free
 }
 {
  bone1Name=dyn_jacket_d_mstr
  bone2Name=spine_03
  swing1LimitDegrees=45
  swing1Motion=Free
  swing2LimitDegrees=45
  swing2Motion=Free
  twistLimitDegrees=45
  twistMotion=Free
 }
 {
  bone1Name=dyn_jacket_e_mstr
  bone2Name=spine_03
  swing1LimitDegrees=45
  swing1Motion=Free
  swing2LimitDegrees=45
  swing2Motion=Free
  twistLimitDegrees=45
  twistMotion=Free
 }
 {
  bone1Name=dyn_jacket_f_mstr
  bone2Name=spine_03
  swing1LimitDegrees=45
  swing1Motion=Free
  swing2LimitDegrees=45
  swing2Motion=Free
  twistLimitDegrees=45
  twistMotion=Free
 }
 {
  bone1Name=thigh_r
  bone2Name=pelvis
  swing1LimitDegrees=45
  swing1Motion=Free
  swing2LimitDegrees=45
  swing2Motion=Free
  twistLimitDegrees=45
  twistMotion=Free
 }
 {
  bone1Name=dyn_jacket_c_03
  bone2Name=dyn_jacket_d_03
  swing1LimitDegrees=45
  swing1Motion=Free
  swing2LimitDegrees=45
  swing2Motion=Free
  twistLimitDegrees=45
  twistMotion=Free
 }
 {
  bone1Name=upperarm_l
  bone2Name=spine_03
  swing1LimitDegrees=45
  swing1Motion=Free
  swing2LimitDegrees=45
  swing2Motion=Free
  twistLimitDegrees=45
  twistMotion=Free
 }
]
