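# Animation Blueprint for the Distant Echo Pilot Style male cosmetic body, adding secondary motion to the parent mesh's copied pose.
# Deform and control rigs feed weighted constraints and additive offsets on belt attachments, then collar AnimDynamics and three RigidBody passes.
# The update event fades out the first RigidBody in front end and disables JacketControl during emotes, front end or DBNO.
# The holster AnimDynamics node and a second constraint chain are unconnected.
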AnimBlueprint /BRCosmetics/Characters/Player/Male/Medium/Bodies/M_MED_DistantEchoPilot_Style/Meshes/M_MED_DistantEchoPilot_Style_AnimBP
target_skeleton=&/M_MED_DistantEchoPilot_Style_Skeleton
graphs [
 AnimGraph {nodes [
   Anim Dynamics - Bone: dyn_holster(Alpha=1, ExternalForce="0,0,0") {
    angularSpringConstant=5
    bAngularSpring=true
    bUseGravityOverride=true
    simulationSpace=BoneRelative
    alphaBoolBlend {blendOption=Linear}
    boundBone {boneName=dyn_holster}
    componentAppliedLinearAccClamp=<10000 10000 10000>
    componentLinearAccScale=<1 1 1>
    componentLinearVelScale=<1 1 1>
    gravityOverride {z=-980}
    physicsBodyDefinitions [{
      boundBone {boneName=dyn_holster}
      boxExtents=<4 2 2>
      constraintSetup {angularLimitsMax=<40 0 2>, angularLimitsMin=<-40 0 -2>, angularTarget=<0 0 0>}
      localJointOffset=<0 0 -1>
     }]
    relativeSpaceBone {boneName=pelvis}
   }
   n1 = "Transform (Modify) Bone - Bone: dyn_tube"(Alpha=1, Translation="0.5,-1.5,0.6", Rotation="-6.398,10,10.141", Scale="1,1,1") {
    rotationMode=BMM_Additive
    rotationSpace=BCS_BoneSpace
    translationMode=BMM_Additive
    translationSpace=BCS_BoneSpace
    alphaBoolBlend {blendOption=Linear}
    boneToModify {boneName=dyn_tube}
    rotation=<rot -6.398 10 10.141>
    translation=<0.5 -1.5 0.6>
   }
   n2 = "Constraint - dyn_box02 (PP)"!(ComponentPose=@n1, Alpha=1, ConstraintWeights_0=0.8, ConstraintWeights_1:float=0.2) -> (Pose) {
    alphaBoolBlend {blendOption=Linear}
    boneToModify {boneName=dyn_box02}
    constraintSetup [{offsetOption=None, targetBone {boneName=dyn_tube}}, {targetBone {boneName=deform_groin_r}}]
    constraintWeights [1,1]
   }
   n3 = "Constraint - dyn_box03 (PPP)"!(ComponentPose=@n2, Alpha=1, ConstraintWeights_0=0.15, ConstraintWeights_1:float=0.7, ConstraintWeights_2:float=0.15) -> (Pose) {
    alphaBoolBlend {blendOption=Linear}
    boneToModify {boneName=dyn_box03}
    constraintSetup [{targetBone {boneName=deform_groin_r}}, {targetBone {boneName=dyn_box02}}, {targetBone {boneName=thigh_twist_01_r}}]
    constraintWeights [1,1,1]
   }
   n4 = "Transform (Modify) Bone - Bone: dyn_box03"(ComponentPose=@n3, Alpha=1, Translation="1.380,-0.412,-0.571", Rotation="0.511,-22.881,0.646", Scale="1,1,1") {
    rotationMode=BMM_Additive
    rotationSpace=BCS_BoneSpace
    translationMode=BMM_Additive
    translationSpace=BCS_BoneSpace
    alphaBoolBlend {blendOption=Linear}
    boneToModify {boneName=dyn_box03}
    rotation=<rot 0.511 -22.881 0.646>
    translation=<1.38 -0.412 -0.571>
   }
   "Transform (Modify) Bone - Bone: dyn_box02"(ComponentPose=@n4, Alpha=1, Translation="0.4,2,-0.889", Rotation="0.490,0,0.162", Scale="1,1,1") {
    rotationMode=BMM_Additive
    rotationSpace=BCS_BoneSpace
    translationMode=BMM_Additive
    translationSpace=BCS_BoneSpace
    alphaBoolBlend {blendOption=Linear}
    boneToModify {boneName=dyn_box02}
    rotation {pitch=0.49, roll=0.162}
    translation=<0.4 2 -0.889>
   }
   n5 = Control Rig(Source=Copy Pose From Mesh() {bUseAttachedParent=true}) {
    alphaBoolBlend {blendOption=Linear}
    controlRigAssetReference {blueprintRigClass=/BRCosmetics/Characters/Player/Male/Medium/Bodies/M_MED_DistantEchoPilot/Meshes/M_DistantEchoPilot_DeformRig.M_DistantEchoPilot_DeformRig_C}
   }
   n6 = Control Rig!(Source=@n5, Alpha, IsCrouchingCtrl:bool=Get bIsCrouching(), JacketCtrl:bool=Get JacketControl() -> (JacketControl:bool)) -> (Pose) {
    alphaBoolBlend {blendOption=Linear}
    controlRigAssetReference {blueprintRigClass=/BRCosmetics/Characters/Player/Male/Medium/Bodies/M_MED_DistantEchoPilot/Meshes/M_MED_DistantEchoPilot_CtrlRig.M_MED_DistantEchoPilot_CtrlRig_C}
   }
   n7 = "Constraint - dyn_mstr_holster (PP)"!(ComponentPose=Local To Component(LocalPose=@n6), Alpha=1, ConstraintWeights_0=0.2, ConstraintWeights_1:float=0.8) -> (Pose) {
    alphaBoolBlend {blendOption=Linear}
    boneToModify {boneName=dyn_mstr_holster}
    constraintSetup [{targetBone {boneName=thigh_r}}, {targetBone {boneName=thigh_twist_01_r}}]
    constraintWeights [1,1]
   }
   n8 = "Constraint - dyn_mstr_comm (PPPPP)"!(
    ComponentPose=@n7
    Alpha=1
    ConstraintWeights_0=0
    ConstraintWeights_1:float=0.38
    ConstraintWeights_2:float=0.3
    ConstraintWeights_3:float=0.3
    ConstraintWeights_4:float=0.02
   ) -> (Pose) {
    alphaBoolBlend {blendOption=Linear}
    boneToModify {boneName=dyn_mstr_comm}
    constraintSetup [
     {targetBone {boneName=spine_01}}
     {targetBone {boneName=pelvis}}
     {targetBone {boneName=thigh_twist_01_l}}
     {targetBone {boneName=deform_groin_l}}
     {targetBone {boneName=dyn_beltbuckle_mstr}}
    ]
    constraintWeights [1,1,1,1,1]
   }
   n9 = "Constraint - dyn_box01 (PPPP)"!(
    ComponentPose=@n8
    Alpha=1
    ConstraintWeights_0=0.22
    ConstraintWeights_1:float=0.15
    ConstraintWeights_2:float=0.16
    ConstraintWeights_3:float=0.47
   ) -> (Pose) {
    alphaBoolBlend {blendOption=Linear}
    boneToModify {boneName=dyn_box01}
    constraintSetup [
     {targetBone {boneName=pelvis}}
     {targetBone {boneName=thigh_twist_01_l}}
     {targetBone {boneName=deform_groin_l}}
     {targetBone {boneName=dyn_beltbuckle_mstr}}
    ]
    constraintWeights [1,1,1,1]
   }
   n10 = "Constraint - dyn_box02 (PPPP)"!(
    ComponentPose=@n9
    Alpha=1
    ConstraintWeights_0=0.26
    ConstraintWeights_1:float=0.25
    ConstraintWeights_2:float=0.45
    ConstraintWeights_3:float=0.04
   ) -> (Pose) {
    alphaBoolBlend {blendOption=Linear}
    boneToModify {boneName=dyn_box02}
    constraintSetup [
     {targetBone {boneName=pelvis}}
     {targetBone {boneName=thigh_twist_01_r}}
     {targetBone {boneName=deform_groin_r}}
     {targetBone {boneName=dyn_beltbuckle_mstr}}
    ]
    constraintWeights [1,1,1,1]
   }
   n11 = "Constraint - dyn_box03 (PPP)"(ComponentPose=@n10, Alpha=1, ConstraintWeights_0=0.39, ConstraintWeights_1=0.36, ConstraintWeights_2=0.25) {
    alphaBoolBlend {blendOption=Linear}
    boneToModify {boneName=dyn_box03}
    constraintSetup [{targetBone {boneName=deform_groin_r}}, {targetBone {boneName=pelvis}}, {targetBone {boneName=thigh_twist_01_r}}]
    constraintWeights [1,1,1]
   }
   n12 = "Constraint - dyn_tube (PPPP)"!(
    ComponentPose=@n11
    Alpha=1
    ConstraintWeights_0=0.1
    ConstraintWeights_1:float=0.1
    ConstraintWeights_2:float=0.18
    ConstraintWeights_3:float=0.64
   ) -> (Pose) {
    alphaBoolBlend {blendOption=Linear}
    boneToModify {boneName=dyn_tube}
    constraintSetup [
     {targetBone {boneName=deform_groin_r}}
     {targetBone {boneName=pelvis}}
     {targetBone {boneName=thigh_twist_01_r}}
     {targetBone {boneName=dyn_beltbuckle_mstr}}
    ]
    constraintWeights [1,1,1,1]
   }
   n13 = "Constraint - dyn_box03 (PPPP)"!(
    ComponentPose=@n12
    Alpha=1
    ConstraintWeights_0=0.32
    ConstraintWeights_1:float=0.08
    ConstraintWeights_2:float=0.26
    ConstraintWeights_3:float=0.34
   ) -> (Pose) {
    alphaBoolBlend {blendOption=Linear}
    boneToModify {boneName=dyn_box03}
    constraintSetup [
     {offsetOption=None, targetBone {boneName=pelvis}}
     {offsetOption=None, targetBone {boneName=dyn_beltbuckle_mstr}}
     {offsetOption=None, targetBone {boneName=thigh_twist_01_r}}
     {targetBone {boneName=deform_groin_r}}
    ]
    constraintWeights [1,1,1,1]
   }
   n14 = "Constraint - dyn_box02 (PPPP)"(ComponentPose=@n13, Alpha=1, ConstraintWeights_0=0.28, ConstraintWeights_1=0.16, ConstraintWeights_2=0.26, ConstraintWeights_3=0.3) {
    alphaBoolBlend {blendOption=Linear}
    boneToModify {boneName=dyn_box02}
    constraintSetup [
     {offsetOption=None, targetBone {boneName=pelvis}}
     {offsetOption=None, targetBone {boneName=dyn_beltbuckle_mstr}}
     {offsetOption=None, targetBone {boneName=thigh_twist_01_r}}
     {targetBone {boneName=deform_groin_r}}
    ]
    constraintWeights [1,1,1,1]
   }
   n15 = "Transform (Modify) Bone - Bone: dyn_box03"(ComponentPose=@n14, Alpha=1, Translation="-3.971,3.308,4.266", Rotation="-40.856,13.254,27.312", Scale="1,1,1") {
    rotationMode=BMM_Additive
    translationMode=BMM_Additive
    alphaBoolBlend {blendOption=Linear}
    boneToModify {boneName=dyn_box03}
    rotation=<rot -40.856 13.254 27.312>
    translation=<-3.971 3.308 4.266>
   }
   n16 = "Transform (Modify) Bone - Bone: dyn_box02"(ComponentPose=@n15, Alpha=1, Translation="-2.344,3.708,4.503", Rotation="-32.090,36.118,1.468", Scale="1,1,1") {
    rotationMode=BMM_Additive
    translationMode=BMM_Additive
    alphaBoolBlend {blendOption=Linear}
    boneToModify {boneName=dyn_box02}
    rotation=<rot -32.09 36.118 1.468>
    translation=<-2.344 3.708 4.503>
   }
   n17 = "Transform (Modify) Bone - Bone: dyn_tube"(ComponentPose=@n16, Alpha=1, Translation="1.713,2.584,0.8", Rotation="1.043,-8.603,0.268", Scale="1,1,1") {
    rotationMode=BMM_Additive
    translationMode=BMM_Additive
    alphaBoolBlend {blendOption=Linear}
    boneToModify {boneName=dyn_tube}
    rotation=<rot 1.043 -8.603 0.268>
    translation=<1.713 2.584 0.8>
   }
   n18 = Anim Dynamics - Bone: dyn_collar_fl(ComponentPose=@n17, Alpha=0.3, ExternalForce="0,0,0") {
    angularDampingOverride=1
    angularSpringConstant=1500
    bAngularSpring=true
    bOverrideAngularDamping=true
    bUseGravityOverride=true
    linearSpringConstant=10
    simulationSpace=RootRelative
    alphaBoolBlend {blendOption=Linear}
    boundBone {boneName=dyn_collar_fl}
    componentAppliedLinearAccClamp=<10000 10000 10000>
    componentLinearAccScale=<20 20 20>
    componentLinearVelScale=<20 20 20>
    gravityOverride {z=-980}
    physicsBodyDefinitions [{
      boundBone {boneName=dyn_collar_fl}
      boxExtents=<3 8 8>
      constraintSetup {angularLimitsMax=<5 90 0>, angularLimitsMin=<-5 -20 0>, angularTarget=<0 0 0>}
      localJointOffset=<1 0 -1>
     }]
    relativeSpaceBone {boneName=spine_05}
   }
   n19 = Anim Dynamics - Bone: dyn_collar_fr(ComponentPose=@n18, Alpha=0.3, ExternalForce="0,0,0") {
    angularDampingOverride=1
    angularSpringConstant=1500
    bAngularSpring=true
    bOverrideAngularDamping=true
    bUseGravityOverride=true
    linearSpringConstant=10
    simulationSpace=RootRelative
    alphaBoolBlend {blendOption=Linear}
    boundBone {boneName=dyn_collar_fr}
    componentAppliedLinearAccClamp=<10000 10000 10000>
    componentLinearAccScale=<20 20 20>
    componentLinearVelScale=<20 20 20>
    gravityOverride {z=-980}
    physicsBodyDefinitions [{
      boundBone {boneName=dyn_collar_fr}
      boxExtents=<3 12 12>
      constraintSetup {angularLimitsMax=<5 90 0>, angularLimitsMin=<-5 -20 0>, angularTarget=<0 0 0>}
      localJointOffset=<1 0 -1>
     }]
    relativeSpaceBone {boneName=spine_05}
   }
   n20 = Anim Dynamics - Bone: dyn_collar_fr02(ComponentPose=@n19, Alpha=0.4, ExternalForce="0,0,0") {
    angularDampingOverride=1
    angularSpringConstant=1500
    bAngularSpring=true
    bOverrideAngularDamping=true
    bUseGravityOverride=true
    linearSpringConstant=10
    simulationSpace=RootRelative
    alphaBoolBlend {blendOption=Linear}
    boundBone {boneName=dyn_collar_fr02}
    componentAppliedLinearAccClamp=<10000 10000 10000>
    componentLinearAccScale=<20 20 20>
    componentLinearVelScale=<20 20 20>
    gravityOverride {z=-980}
    physicsBodyDefinitions [{
      boundBone {boneName=dyn_collar_fr02}
      boxExtents=<5 30 30>
      constraintSetup {angularLimitsMax=<5 100 0>, angularLimitsMin=<-5 -20 0>, angularTarget=<0 0 0>}
      localJointOffset=<5 0 -5>
     }]
    relativeSpaceBone {boneName=spine_05}
   }
   n21 = RigidBody(ComponentPose=@n20, Alpha=Get FrontEndAlpha() -> (FrontEndAlpha:double)) {
    overridePhysicsAsset=/BRCosmetics/Characters/Player/Male/Medium/Bodies/M_MED_DistantEchoPilot/Meshes/M_MED_DistantEchoPilot_Physics
    simulationSpace=BaseBoneSpace
    alphaBoolBlend {blendOption=Linear}
    baseBoneRef {boneName=spine_04}
    componentAppliedLinearAccClamp=<1000 1000 1000>
    componentLinearAccScale=<0.05 0.05 0.05>
    componentLinearVelScale=<0.05 0.05 0.05>
    simSpaceSettings {velocityScaleZ=0.5, worldAlpha=0.2}
   }
   n22 = RigidBody(ComponentPose=@n21) {
    overridePhysicsAsset=&/M_MED_DistantEchoPilot_Style_FE_Physics
    simulationSpace=BaseBoneSpace
    alphaBoolBlend {blendOption=Linear}
    baseBoneRef {boneName=spine_04}
    componentAppliedLinearAccClamp=<1000 1000 1000>
    componentLinearAccScale=<0.05 0.05 0.05>
    componentLinearVelScale=<0.05 0.05 0.05>
    simSpaceSettings {velocityScaleZ=0.5, worldAlpha=0.2}
   }
   n23 = RigidBody(ComponentPose=@n22) {
    overridePhysicsAsset=/BRCosmetics/Characters/Player/Male/Medium/Bodies/M_MED_DistantEchoPilot/Meshes/M_MED_DistantEchoPilot_DBNO_Physics
    simulationSpace=BaseBoneSpace
    alphaBoolBlend {blendOption=Linear}
    baseBoneRef {boneName=spine_04}
    componentAppliedLinearAccClamp=<1000 1000 1000>
    componentLinearAccScale=<0.05 0.05 0.05>
    componentLinearVelScale=<0.05 0.05 0.05>
    simSpaceSettings {velocityScaleZ=0.5, worldAlpha=0.2}
   }
   Output Pose(Result=Component To Local(ComponentPose=@n23)) {layerGroup=None}
  ]}
 EventGraph {nodes [
   Event BlueprintUpdateAnimation()
   > Set PouchDefault(PouchDefault:double=SelectFloat(A=0, B=1, bPickA=Get bIsCrouching())) -> (Output_Get:double)
   > Set FrontEndAlpha(FrontEndAlpha:double=SelectFloat(A=0, B=1, bPickA=Get bIsInFrontEnd())) -> (Output_Get:double)
   n1 = OR Boolean!(A=Get bIsPlayingEmote(), B=Get bIsInFrontEnd(), C:bool=Get bIsDBNO()) -> (ReturnValue)
   > Set JacketControl(JacketControl:bool="Equal (Float)"(A=SelectFloat(A=0, B=1, bPickA=@n1), B=1)) -> (Output_Get:bool)
   NOT Boolean(A=false)
   Get bIsCrouching()
  ]}
]
variables [
 FrontEndAlpha {type=number}
 JacketControl {type=boolean}
 PouchDefault {type=number}
]
